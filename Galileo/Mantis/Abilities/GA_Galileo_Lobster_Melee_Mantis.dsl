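; Mantis melee attack ability for the Galileo Lobster weapon, toggling its audio and reticle and ending if the player starts surface swimming.
; Hits on pawns with the melee-ability tag play a clash cue.
; Optional prototype stagger sends opponents a cancel event before applying the stagger effect, then launches them away.
; While active, dedicated servers relax movement correction and optionally acceleration tolerance, skipping changes during Grinding, via test toggles marked for removal.
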(event Ability|EventActivateAbility
  (bind _returnvalue (Ability|GetCurrentSourceWeapon))
  (bind _asb_athena_galileo_lobster_parent (Utilities|Casting|CastToB_Athena_Galileo_Lobster_Parent _returnvalue))
  (Development|PrintString "Activate Mantis Attack" false false "(R=1.000000,G=0.492313,B=0.000000,A=1.000000)")
  (bind _asfort_player_pawn (Utilities|Casting|CastToFortPlayerPawn (Ability|GetActivatingPawn)))
  (bind _output_get (Variables|Default|SetFort%20Player%20Pawn _asfort_player_pawn))
  (Character|UnCrouch _output_get)
  (Class|BAthenaGalileoLobsterParent|ToggleDynamicAudio _asb_athena_galileo_lobster_parent false)
  (Class|BAthenaGalileoLobsterParent|ReticleStatusDisplay _asb_athena_galileo_lobster_parent true)
  ;[1 1 1] Kill the Ability on SurfaceSwimming
  (Ability|Tasks|WaitGameplayTagAdd "(TagName=\"MovementMode.SurfaceSwimming\")"
    (:then
      (CallFunction|SetSpecialCharacterMovement true)
      ;[1 1 1] Enable Melee Attack Stagger/Cancel Prototype
      (bind _output_get_2 (Variables|Stagger|SetStaggerMelee (Math|Conversions|ToBoolean(Integer) (Math|Float|Truncate (ScalableFloat|GetValueatLevel (Variables|Stagger|GetEnableStagger))))))
      (if _output_get_2
        (Ability|Tasks|WaitGameplayEvent "(TagName=\"Event.Melee.Stagger\")"
          (:EventReceived
            (Ability|EndAbility)))))
    (:Added
      (Ability|EndAbility))))

(event Mantis|EventOnMantisTechniqueHit
  (params (TargetDataHandle :type Gameplay%20Ability%20Target%20Data%20Handle%20Structure%20%28by%20ref%29) (ApplicationTag :type Gameplay%20Tag%20Structure))
  (bind _fort_player_pawn (Variables|Default|GetFort%20Player%20Pawn))
  ;[1 1 1] Melee Attack Stagger/Cancel Prototype
  (for _array_element (Ability|TargetData|GetAllActorsfromTargetData TargetDataHandle)
    (bind _asfort_player_pawn (Utilities|Casting|CastToFortPlayerPawn _array_element))
    (bind _returnvalue (GameplayTags|HasMatchingGameplayTag(Message) _asfort_player_pawn (Variables|Default|GetT_MeleeAbility)))
    (if _returnvalue
      (Ability|ExecuteGameplayCueWithParamsOnOwner (Variables|Default|GetT_GCMeleeClash) (Ability|GameplayCue|MakeGameplayCueParameters :NormalizedMagnitude 0.0 :RawMagnitude 0.0 :Location "0,0,0" :Normal "0,0,0" :Instigator _fort_player_pawn :EffectCauser _fort_player_pawn :GameplayEffectLevel 1 :AbilityLevel 1 :bReplicateLocationWhenUsingMinimalRepProxy false)) ; Sword on sword GC 
      (if (Variables|Stagger|GetStaggerMelee)
        (Networking|CanExecuteCosmeticEvents
          (:False
            ;;[1 1 1] End ability for the swinging enemy.  Needs to be done before Stagger ability is granted or this interaction breaks over light latency
            (Ability|SendGameplayEventtoActor _asfort_player_pawn (Variables|Stagger|GetT_StaggerEvent))
            (Ability|ApplyGameplayEffectToTarget TargetDataHandle (Variables|Stagger|GetGE_StaggerEnemy))
            (Pawn|Character|LaunchCharacterJump _asfort_player_pawn (Math|Vector|vector+vector (* (Math|Vector|Normalize (- (Transformation|GetActorLocation _asfort_player_pawn) (Transformation|GetActorLocation _fort_player_pawn))) (ScalableFloat|GetValueatLevel (Variables|Stagger|GetImpulseXY))) (Math|Vector|MakeVector 0.0 0.0 (ScalableFloat|GetValueatLevel (Variables|Stagger|GetImpulseZ)))) true true false true)))))))

(event Ability|EventOnEndAbility
  (params (bWasCancelled :type Boolean))
  (bind _returnvalue (Ability|GetCurrentSourceWeapon))
  (bind _asb_athena_galileo_lobster_parent (Utilities|Casting|CastToB_Athena_Galileo_Lobster_Parent _returnvalue))
  ;[1 1 1] End Ability
  (Class|BAthenaGalileoLobsterParent|ToggleDynamicAudio _asb_athena_galileo_lobster_parent true)
  (Class|BAthenaGalileoLobsterParent|ReticleStatusDisplay _asb_athena_galileo_lobster_parent false)
  (Animation|AnimTrailsNotify (Variables|Default|GetFort%20Player%20Pawn) false)
  (CallFunction|SetSpecialCharacterMovement false))

(event Custom|SetLobsterClientAuthorityEnabled
  (params (bEnabled :type Boolean))
  ;[1 0.5836 0] Events to toggle character movement options. TODO remove this after testing.
  (Variables|Testing|SetbUseClientPosition bEnabled))

(event Custom|SetLobsterHighAccelerationToleranceEnabled
  (params (bEnabled :type Boolean))
  ;[1 0.5836 0] Events to toggle character movement options. TODO remove this after testing.
  (Variables|Testing|SetbUseHighAccelerationTolerance bEnabled))

(fn GetLevelForGameplayEffectContainer
  (params (ApplicationTag :type Gameplay%20Tag%20Structure))
  (return (Math|Float|Truncate (Ability|GetCurrentAbilitySourceLevel))))

(fn SetSpecialCharacterMovement
  (params (bEnabled :type Boolean))
  (if (Networking|IsDedicatedServer)
    (bind _fort_player_pawn (Variables|Default|GetFort%20Player%20Pawn)
      (:then
        (if (not (Utilities|Enum|Equal(Enum) (Utilities|Enum|LiteralenumEFortCustomMovement "Grinding") (Utilities|Enum|BytetoEnumEFortCustomMovement (Variables|/Script/Engine.CharacterMovementComponent|GetCustomMovementMode (Variables|/Script/Engine.Character|GetCharacterMovement _fort_player_pawn)))))
          (Utilities|FlowControl|Sequence
            (:then_0
              ;[1 0.5592 0] Set character movement variables if enabled. TODO change or remove this after testing.
              (if (Variables|Testing|GetbUseClientPosition)
                (Variables|/Script/Engine.CharacterMovementComponent|SetbServerAcceptClientAuthoritativePosition (Variables|GetbEnabled) (Variables|/Script/Engine.Character|GetCharacterMovement _fort_player_pawn))
                (Variables|/Script/Engine.CharacterMovementComponent|SetbIgnoreClientMovementErrorChecksAndCorrection (Variables|GetbEnabled) (Variables|/Script/Engine.Character|GetCharacterMovement _fort_player_pawn))))
            (:then_1
              (if (Variables|Testing|GetbUseHighAccelerationTolerance)
                (bind _asfort_movement_comp_character_athena (Utilities|Casting|CastToFortMovementComp_CharacterAthena (Variables|/Script/Engine.Character|GetCharacterMovement _fort_player_pawn)))
                (NetTolerance|SetUseHighAccelerationTolerance _asfort_movement_comp_character_athena (Variables|GetbEnabled) "Lobster"))))))
      (:else))))
